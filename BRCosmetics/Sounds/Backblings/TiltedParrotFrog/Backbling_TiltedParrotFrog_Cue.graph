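# Reactive sound cue for the TiltedParrotFrog back bling, playing reaction sounds chosen by an integer StatValue parameter.
# After a fixed half-second delay, a switch sends most values to an evenly weighted random pick of four reactions.
# The highest value plays a distinct Final reaction, and the default slot is silent.
# It uses close-range pet attenuation, the Backblings sound class and per-owner concurrency of one voice that stops the oldest.
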SoundCue /BRCosmetics/Sounds/Backblings/TiltedParrotFrog/Backbling_TiltedParrotFrog_Cue
root=Delay_0
nodes [
 {id=Delay_0, children [Switch_0], properties {delayMax=0.5, delayMin=0.5}}
 {id=Random_1, children [WavePlayer_3,WavePlayer_4,WavePlayer_5,WavePlayer_6], properties {weights [1,1,1,1]}}
 {
  id=Switch_0
  children [null,Random_1,Random_1,Random_1,Random_1,Random_1,Random_1,Random_1,WavePlayer_0]
  properties {intParameterName=StatValue}
 }
 {id=WavePlayer_0, properties {soundWaveAssetPtr=&/Backbling_TiltedParrotFrog_React_Final}}
 {id=WavePlayer_3, properties {soundWaveAssetPtr=&/Backbling_TiltedParrotFrog_React_04}}
 {id=WavePlayer_4, properties {soundWaveAssetPtr=&/Backbling_TiltedParrotFrog_React_01}}
 {id=WavePlayer_5, properties {soundWaveAssetPtr=&/Backbling_TiltedParrotFrog_React_02}}
 {id=WavePlayer_6, properties {soundWaveAssetPtr=&/Backbling_TiltedParrotFrog_React_03}}
]
properties {
 attenuationSettings=&Fort_Pets_Close_Athena_Att
 audioPropertiesSheet=/BRCosmetics/Sounds/Settings/PropertySheets/Reactive/APS_Cosmetic_Reactive_Backbling
 bOverrideConcurrency=true
 maxDistance=1800.0001
 soundClassObject=/Game/Sounds/FortSoundClassesAndModes/Fort_Backblings
 volumeMultiplier=0.8
 concurrencyOverrides {bLimitToOwner=true, maxCount=1, resolutionRule=StopOldest}
}
